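# Landscape material layer function for default-quality Hermes grass, building surface attributes, an opacity mask and displacement.
# A world-aligned packed texture supplies detail, specular, roughness and height masks, alongside world-aligned normals, slope/edge masking and four variation noises.
# Those attributes and terrain masks feed the BiomeColor function, whose base color is optionally multiplied by a detail-driven brightness before output.
# Alternative color, hex-tiled sampling and debug branches remain unconnected to the output.
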MaterialFunction /Game/Hermes/Terrain/Landscape/Materials/Layers/QualityLevel/MF_Hermes_Grass_Default
graph {nodes [
  n1 = NamedRerouteUsage()
  n2 = VectorParameter() {parameterName=NG_G2_C1, defaultValue=<rgba 0.0117 0.15 0.0389 1>}
  n3 = VectorParameter() {parameterName=NG_G2_C3, defaultValue=<rgba 0.0622 0.22 0 1>}
  n4 = MaterialFunctionCall!(
   Coordinates
   R_01_Scale
   R_02_Scale
   R_03_Scale
   R_04_Scale
   G_01_Scale
   G_02_Scale=ScalarParameter() {defaultValue=-5550, group=Grass, parameterName=NG_G02Scale}
   G_03_Scale=ScalarParameter() {defaultValue=9500, group=Grass, parameterName=NG_G03Scale}
   G_04_Scale=ScalarParameter() {defaultValue=-6100, group=Grass, parameterName=NG_G04Scale}
  ) {materialFunction=/Game/Athena/Helios/Environments/Landscape/Materials/MaterialFunctions_Utilities/MF_VariationMasks2}
  n5 = Saturate(None=LinearInterpolate(
    A=ScalarParameter() {defaultValue=1, group=Grass, parameterName=NG_Grass2_Noise2_A}
    B=ScalarParameter() {group=Grass, parameterName=NG_Grass2_Noise2_B}
    Alpha=@n4.G_02
   ))
  n6 = Multiply(
   A=Power(Base=@n4.G_03, Exp=ScalarParameter() {defaultValue=1, group=Grass, parameterName=NG_Grass2_Noise3_Pow})
   B=ScalarParameter() {defaultValue=0.5, group=Grass, parameterName=NG_Grass2_Noise3_Mod}
  )
  n7 = Multiply(A=@n4.G_04, B=ScalarParameter() {defaultValue=1, group=Grass, parameterName=NG_Grass2_Noise4_Mod})
  n8 = LinearInterpolate(
   A=LinearInterpolate(
    A=LinearInterpolate(A=@n2.RGB, B=@n3.RGB, Alpha=@n5)
    B=VectorParameter() {parameterName=NG_G2_C4, defaultValue=<rgba 0.1485 0.25 0.025 1>}.RGB
    Alpha=Saturate(None=@n6)
   )
   B=VectorParameter() {parameterName=NG_G2_C5, defaultValue=<rgba 0.039 0.13 0.0065 1>}.RGB
   Alpha=Power(Base=Saturate(None=@n7), Exp=ScalarParameter() {defaultValue=0.5, group=Grass, parameterName=NG_Grass2_Noise4_Pow})
  )
  Add(A=Multiply(A=Constant() {r=0.01}, B=@n1), B=Multiply(A=@n1, B=@n8))
  Constant3Vector() {constant=<rgba 1.1081 1.3 1.047 1>}
  Constant3Vector() {constant=<rgba 0.0117 0.15 0.0389 1>}
  Constant3Vector() {constant=<rgba 0.0622 0.22 0 1>}
  Constant3Vector() {constant=<rgba 0.1485 0.25 0.025 1>}
  Constant3Vector() {constant=<rgba 0.039 0.13 0.0065 1>}
  Constant() {r=945}
  Constant() {r=700}
  Constant() {r=0.95}
  Constant() {r=0.05}
  Constant() {r=9.9}
  Constant() {r=0.25}
  Constant() {r=0.5}
  Constant() {r=0.5}
  Constant() {r=0.3}
  Constant() {r=0.4}
  Constant() {r=5}
  Constant()
  Constant() {r=0.3}
  Constant() {r=0.4}
  n9 = NamedRerouteUsage()
  n10 = ScalarParameter() {defaultValue=0.25, group=Displacement, parameterName=Grass_Displacement_Low}
  n11 = ScalarParameter() {defaultValue=0.5, group=Displacement, parameterName=Grass_Displacement_High}
  n12 = SetMaterialAttributes!(
   MaterialAttributes
   Specular=NamedRerouteUsage()
   Roughness=NamedRerouteUsage()
   Normal=NamedRerouteUsage()
   OpacityMask=@n9
   Displacement=Saturate(None=LinearInterpolate(A=@n10, B=@n11, Alpha=@n9))
  ) {attributeSetTypes [
    #1
    #2
    #3
    #4
    #5
   ]}
  n13 = MaterialFunctionCall!(
   Material=@n12
   GrassEdgeMask=NamedRerouteUsage()
   TerrainMasks_QuadA=FunctionInput(Preview=Constant()) {bUsePreviewValueAsDefault=true, inputName=TerrainMasks_QuadA, inputType=FunctionInput_Vector4, sortPriority=-400}
   TerrainMasks_QuadB=FunctionInput(Preview=Constant()) {bUsePreviewValueAsDefault=true, inputName=TerrainMasks_QuadB, inputType=FunctionInput_Vector4, sortPriority=-300}
   TerrainMasks_QuadC=FunctionInput(Preview=Constant()) {bUsePreviewValueAsDefault=true, inputName=TerrainMasks_QuadC, inputType=FunctionInput_Vector4, sortPriority=-200}
   TerrainMasks_QuadD=FunctionInput(Preview=Constant()) {bUsePreviewValueAsDefault=true, inputName=TerrainMasks_QuadD, inputType=FunctionInput_Vector4, sortPriority=-100}
   GrassNoise=NamedRerouteUsage()
   VariationMasks Mobile
  ) {materialFunction=/Game/Hermes/Terrain/Landscape/Materials/Layers/BiomeColor/MF_Hermes_Grass_BiomeColor}
  n14 = GetMaterialAttributes(None=@n13.Result) {attributeGetTypes [#6]}
  n15 = ScalarParameter() {defaultValue=0.2, group=Grass, parameterName=Grass2_BCD_LerpA}
  n16 = ScalarParameter() {defaultValue=1.25, group=Grass, parameterName=Grass2_BCD_LerpB}
  n17 = Power(Base=NamedRerouteUsage(), Exp=ScalarParameter() {defaultValue=3, group=Grass, parameterName=Grass2_BCD_Pow11})
  n18 = Multiply(A=Saturate(None=LinearInterpolate(A=@n15, B=@n16, Alpha=@n17)), B=@n14.BaseColor)
  FunctionOutput(None=StaticSwitch(
    True=SetMaterialAttributes!(MaterialAttributes=@n14.MaterialAttributes, BaseColor=@n18) {attributeSetTypes [#6]}
    False=@n13.Result
    Value=StaticBoolParameter() {defaultValue=true, group=Grass, parameterName=UseGrassBaseColorDetail}
   ))
  n19 = MaterialFunctionCall!(Scale=ScalarParameter() {defaultValue=945, group=Grass, parameterName=Grass_WorldAlignedScale, sortPriority=-1000}) {materialFunction=/Game/Athena/Apollo/Environments/Landscape/Materials/Biome/CommonMaterials/MF_SimpleWorldUV}
  n20 = ScalarParameter() {defaultValue=10, group=Grass, parameterName=Grass_WorldAlignedScalePeriodic, sortPriority=-1000}
  n21 = WorldPosition() {worldPositionShaderOffset=WPT_CameraRelativeNoOffsets}
  LinearInterpolate(
   A=@n19.Result
   B=Multiply(
    A=MaterialFunctionCall!(Scale=@n20, Camera Relative World Position=@n21.XYZ) {materialFunction=/Engine/Functions/Engine_MaterialFunctions03/Texturing/PeriodicWorldPosition}.XY
    B=Constant2Vector() {g=-1, r=-1}
   )
   Alpha=ScalarParameter() {defaultValue=1, group=Grass, parameterName=Grass_WorldAlignedScalePeriodic_DEBUG, sortPriority=-1000}
  )
  LinearInterpolate(
   A=@n18
   B=TextureSample(UVs=@n19.Result) {texture=/Game/Effects/Fort_Effects/Emotes/DontBeASquare/T_SquareOutline_alpha}.RGB
   Alpha=ScalarParameter() {defaultValue=1, group=Grass, parameterName=Grass_WorldAlignedScalePeriodicSquare_DEBUG, sortPriority=-1000}
  )
  n22 = Multiply(
   A=Power(Base=NamedRerouteUsage(), Exp=ScalarParameter() {defaultValue=1, group=Grass, parameterName=Grass2_HNoise_Pow1})
   B=ScalarParameter() {defaultValue=2, group=Grass, parameterName=Grass2_HNoise_Mult}
  )
  n23 = MaterialFunctionCall() {materialFunction=&MF_VariationMasks2}
  n24 = Saturate(None=Multiply(A=@n23.G_03, B=ScalarParameter() {defaultValue=3, group=Grass, parameterName=Grass2_Noise4_Mult23}))
  n25 = OneMinus(None=@n23.G_04)
  n26 = ScalarParameter() {defaultValue=3, group=Grass, parameterName=Grass2_Noise4_Pow22}
  n27 = Saturate(None=Multiply(A=Power(Base=@n25, Exp=@n26), B=ScalarParameter() {defaultValue=3, group=Grass, parameterName=Grass2_Noise4_Mult22}))
  LinearInterpolate(
   A=LinearInterpolate(A=ScalarParameter() {defaultValue=0.4, group=Grass, parameterName=Grass2_BaseNoiseLowAmt}, Alpha=Saturate(None=@n22))
   Alpha=Saturate(None=Add(A=@n24, B=@n27))
  )
  n28 = NamedRerouteUsage()
  n29 = ScalarParameter() {defaultValue=2, group=Grass, parameterName=NG_BCD_Power}
  n30 = Saturate(None=Multiply(A=@n8, B=Power(Base=@n28, Exp=@n29)))
  n31 = ScalarParameter() {defaultValue=1.5, parameterName=NG_TipMod}
  LinearInterpolate(
   A=@n30
   B=Saturate(None=Multiply(A=@n30, B=@n31))
   Alpha=Power(Base=NamedRerouteUsage(), Exp=ScalarParameter() {defaultValue=3, group=Grass, parameterName=NG_Grass2_Tip_HeightPower})
  )
  n32 = Multiply(A=OneMinus(None=NamedRerouteUsage())) {constB=0.75}
  n33 = LinearInterpolate(
   A=ScalarParameter() {defaultValue=0.95, group=Grass, parameterName=GrassMaxRough}
   B=ScalarParameter() {defaultValue=0.6, group=Grass, parameterName=GrassMinRough}
   Alpha=NamedRerouteUsage()
  )
  LinearInterpolate(A=Saturate(None=Add(A=@n32) {constB=0.35}), B=@n33, Alpha=ScalarParameter() {group=Grass, parameterName=Debug_GrassRoughness})
  n34 = TextureObject() {samplerType=SAMPLERTYPE_Masks, texture=/Game/Environments/Apollo/Sets/Terrain/Textures/T_Apollo_Grass_D_02}
  n35 = Divide(
   A=WorldPosition() {worldPositionShaderOffset=WPT_ExcludeAllShaderOffsets}.XY
   B=ScalarParameter() {defaultValue=700, group=Grass, parameterName=Grass_WorldAlignedHexBasedScale, sortPriority=-1000}
  )
  n36 = StaticBool()
  MaterialFunctionCall!(
   Texture=@n34
   LandscapeCoordinates=@n35
   LowEndTiling
   HexGridGradientSharpness
   HexGridSize
   HexGridUVInput
   MinScale
   MaxScale
   MinOffset
   MaxOffset
   "MaxRotation(Degrees)"
   "MinRotation(Degrees)"
   forceLowQuality=@n36
  ) {materialFunction=/Game/Athena/Asteria/Environments/Landscape/Materials/MaterialFunctions/UtilityFunctions/Landscape_TextureSample}
  n37 = VectorParameter() {parameterName=G2_C1, defaultValue=<rgba 0.0845 0.1732 0.0065 1>}
  n38 = LinearInterpolate(A=@n37.RGB, B=VectorParameter() {parameterName=G2_C2, defaultValue=<rgba 0.0651 0.1137 0.0105 1>}.RGB, Alpha=NamedRerouteUsage())
  n39 = Desaturation(None=@n37.RGB, Fraction=ScalarParameter() {defaultValue=0.14, group=Grass, parameterName=G2_C3_Desaturation})
  n40 = NamedRerouteUsage()
  n41 = LinearInterpolate(
   A=VectorParameter() {parameterName=G2_C5, defaultValue=<rgba 0.0503 0.13 0.0125 1>}.RGB
   B=LinearInterpolate(
    A=LinearInterpolate(A=@n38, B=@n39, Alpha=@n40)
    B=VectorParameter() {parameterName=G2_C4, defaultValue=<rgba 0.1263 0.144 0.0199 1>}.RGB
    Alpha=NamedRerouteUsage()
   )
   Alpha=NamedRerouteUsage()
  )
  Multiply(A=@n41) {desc="For the grass meshes, maybe just use the raw color or blend to the tip color version in the distance?"}
  NamedRerouteDeclaration(None=MaterialFunctionCall(X=NamedRerouteUsage(), Y=NamedRerouteUsage(), Z=NamedRerouteUsage(), A=NamedRerouteUsage()) {materialFunction=&MakeFloat4}.Result) {name=GrassNoiseV4, nodeColor=<rgba 0.5 0 1 1>}
  n42 = Constant()
  n43 = Constant() {r=1.5}
  NamedRerouteDeclaration(None=Saturate(None=LinearInterpolate(A=@n42, B=@n43, Alpha=@n23.G_01))) {name=Noise01, nodeColor=<rgba 0.5 0 1 1>}
  n44 = MaterialFunctionCall(
   Transition Phase=MaterialFunctionCall() {materialFunction=/Game/Athena/Helios/Environments/Landscape/Materials/MaterialFunctions_Utilities/MF_CliffSlopeMask}.Result
   Height Texture=Constant() {r=0.5}
  ) {materialFunction=&HeightLerp}
  n45 = Constant() {r=3}
  n46 = Saturate(None=Multiply(A=Power(Base=@n44.Lerp Alpha No Contrast, Exp=@n45), B=Constant() {r=6}))
  NamedRerouteDeclaration(None=Multiply(A=OneMinus(None=@n46), B=NamedRerouteUsage())) {name=ModifiedHeightMap, nodeColor=<rgba 0.5 0 1 1>}
  n47 = TextureSample(UVs=@n19.Result, Tex=@n34) {samplerSource=SSM_Wrap_WorldGroupSettings, samplerType=SAMPLERTYPE_Masks, texture=&T_WhiteMask}
  n48 = MaterialFunctionCall(Float4=@n47.RGBA) {materialFunction=&BreakOutFloat4Components}
  NamedRerouteDeclaration(None=@n48.G) {name=GrassSpecularMask, nodeColor=<rgba 0.5 0 1 1>}
  NamedRerouteDeclaration(None=LinearInterpolate(
    A=ScalarParameter() {defaultValue=0.05, group=Grass, parameterName=GrassMinSpec_H}
    B=ScalarParameter() {defaultValue=0.5, group=Grass, parameterName=GrassMaxSpec_H}
    Alpha=NamedRerouteUsage()
   )) {name=GrassSpecular, nodeColor=<rgba 0.5 0 1 1>}
  NamedRerouteDeclaration(None=@n41) {name=GrassBaseColor, nodeColor=<rgba 0.5 0 1 1>}
  NamedRerouteDeclaration(None=@n48.B) {name=GrassRoughnessMask, nodeColor=<rgba 0.5 0 1 1>}
  n49 = Multiply(A=NamedRerouteUsage(), B=Constant() {r=0.15})
  NamedRerouteDeclaration(None=Saturate(None=Add(A=@n49, B=@n33))) {name=GrassRoughness, nodeColor=<rgba 0.5 0 1 1>}
  n50 = Constant()
  n51 = Constant()
  n52 = VertexNormalWS()
  n53 = Power(Base=Abs(None=ComponentMask(None=@n52) {b=true}), Exp=ScalarParameter() {defaultValue=1.5, group=Grass, parameterName=GrassSlope1_Pow})
  n54 = Max(
   A=Multiply(
    A=FunctionInput(Preview=@n51) {bUsePreviewValueAsDefault=true, inputName=Grass_EdgeMask, inputType=FunctionInput_Scalar, sortPriority=16}
    B=ScalarParameter() {defaultValue=2, group=Grass, parameterName=GrassEdgeSpread}
   )
   B=Multiply(A=OneMinus(None=@n53), B=ScalarParameter() {defaultValue=1.25, group=Grass, parameterName=GrassSlope1_Mod})
  )
  n55 = FunctionInput(Preview=StaticBool()) {
   bUsePreviewValueAsDefault=true
   desc=Setting this to true will ignore Grass_EdgeMask input
   description=Setting this to true will ignore Grass_EdgeOnly input
   inputName=UseCustomSlopeMask
   inputType=FunctionInput_StaticBool
   sortPriority=13
  }
  NamedRerouteDeclaration(None=StaticSwitch(
    True=FunctionInput(Preview=@n50) {bUsePreviewValueAsDefault=true, inputName=CustomSlopeMask, inputType=FunctionInput_Scalar, sortPriority=14}
    False=Saturate(None=@n54)
    Value=@n55
   )) {name=GrassEdgeSlopeMask, nodeColor=<rgba 0.5 0 1 1>}
  n56 = TextureObject() {samplerType=SAMPLERTYPE_Normal, texture=/Game/Environments/Apollo/Sets/Terrain/Textures/T_Apollo_Dirt_N_02b}
  n57 = TextureSample(UVs=@n19.Result, Tex=@n56) {samplerSource=SSM_Wrap_WorldGroupSettings, samplerType=SAMPLERTYPE_Normal, texture=&FlatNormalBlue}
  n58 = FunctionInput(Preview=StaticBool()) {
   bUsePreviewValueAsDefault=true
   description="Default = False"
   inputName=UseTangentNormals
   inputType=FunctionInput_StaticBool
   sortPriority=1
  }
  n59 = ScalarParameter() {defaultValue=0.4, group=Grass, parameterName=GrassNormalAmount}
  NamedRerouteDeclaration(None=MaterialFunctionCall!(Normals=@n57.RGB, TransformToTangentSpace=@n58, Normal Strength=@n59) {materialFunction=/Game/Athena/Apollo/Environments/Landscape/Materials/Biome/CommonMaterials/MF_HQ_WorldAlignedNormal}.Result) {name=BladeNormals, nodeColor=<rgba 0.5 0 1 1>}
  NamedRerouteDeclaration(None=@n48.A) {name=HeightMapFromTexture, nodeColor=<rgba 0.5 0 1 1>}
  NamedRerouteDeclaration(None=@n48.R) {name=BaseColorDetail, nodeColor=<rgba 1 0.0703 0 1>}
  n60 = Constant() {r=1.5}
  n61 = Constant() {r=-9}
  NamedRerouteDeclaration(None=Saturate(None=LinearInterpolate(A=@n60, B=@n61, Alpha=@n23.G_02))) {name=Noise02, nodeColor=<rgba 0.5 0 1 1>}
  n62 = Constant() {r=-0.25}
  n63 = Constant() {r=1.5}
  n64 = NamedRerouteDeclaration(None=Saturate(None=LinearInterpolate(A=@n62, B=@n63, Alpha=@n23.G_03))) {name=Noise03_PreMult, nodeColor=<rgba 0.5 0 1 1>}
  n65 = Constant() {r=0.65}
  NamedRerouteDeclaration(None=Saturate(None=Multiply(A=@n64, B=@n65))) {name=Noise03, nodeColor=<rgba 0.5 0 1 1>}
  n66 = Constant() {r=4}
  NamedRerouteDeclaration(None=Saturate(None=Multiply(A=@n23.G_04, B=@n66))) {name=Noise04, nodeColor=<rgba 0.5 0 1 1>}
  OneMinus(None=FunctionInput(Preview=Constant()) {bUsePreviewValueAsDefault=true, inputName=LayerWeight, inputType=FunctionInput_Scalar, sortPriority=-500})
  Saturate(None=LinearInterpolate(
    A=ScalarParameter() {defaultValue=0.1, group=Grass, parameterName=GrassMinSpec}
    B=ScalarParameter() {defaultValue=1, group=Grass, parameterName=GrassMaxSpec}
    Alpha=NamedRerouteUsage()
   ))
  n67 = VectorParameter() {group=Grass, parameterName=Grass2_ModColorTemp, defaultValue=<rgba 1 1 1 1>}
  Saturate(None=Multiply(B=@n67.RGB))
  Saturate(None=LinearInterpolate(
    A=ScalarParameter() {group=Grass, parameterName=NG_Grass2_Noise3_A}
    B=ScalarParameter() {defaultValue=1, group=Grass, parameterName=NG_Grass2_Noise3_B}
    Alpha=@n4.G_03
   ))
  SetMaterialAttributes!(
   MaterialAttributes
   BaseColor=MaterialFunctionCall!(Vector3=Multiply(B=@n67.RGB), MaximumNumberOfDigits, UVs, DebugTextLocation RG_UpperRight BA_LowerLeft, Component Spacing) {materialFunction=/Engine/Functions/Engine_MaterialFunctions02/Utility/DebugFloat3Values}.GreyScaleOutput
   Specular=Constant()
   Roughness
  ) {attributeSetTypes [#6, #1, #2]}
  n68 = LinearInterpolate(
   A=Constant3Vector() {constant=<rgba 0 0 1 0>}
   B=ComponentMask(None=MaterialFunctionCall!(
     Texture=@n56
     UseMipBias
     MipBias
     LowEndTiling
     LandscapeCoordinates=@n35
     HexGridSize
     HexGridUVInput
     HexGridGradientSharpness
     MaxScale
     MinScale
     MaxOffset
     MinOffset
     "MaxRotation(Degrees)"
     "MinRotation(Degrees)"
     forceLowQuality=@n36
    ) {materialFunction=/Game/Athena/Asteria/Environments/Landscape/Materials/MaterialFunctions/UtilityFunctions/Landscape_NormalTextureSample}.Result) {b=true, g=true, r=true}
   Alpha=@n59
  )
  StaticSwitch(True=@n68, False=Transform(None=@n68), Value=@n58)
 ]}
