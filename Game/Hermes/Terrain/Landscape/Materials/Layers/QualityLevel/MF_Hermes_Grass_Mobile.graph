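# Mobile grass layer material function for the Hermes landscape, with landscape, static mesh and grass foliage blend outputs.
# Base color blends tinted parameters by masks; a world-XY projected packed texture supplies blade normals and height.
# A color-only material feeds the foliage output; a fuller attribute set feeds the others.
# Both pass through mobile Fresnel, skipped at Medium quality on the default output, then biome coloring driven by terrain and variation mask inputs.
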MaterialFunction /Game/Hermes/Terrain/Landscape/Materials/Layers/QualityLevel/MF_Hermes_Grass_Mobile
graph {
 properties {description=Grass Edge}
 nodes [
  n1 = NamedRerouteUsage()
  n2 = LinearInterpolate(
   A=VectorParameter() {group=Mobile Grass, parameterName=Mobile Grass Variation 1, defaultValue=<rgba 0.4834 0.5048 0.5156 1>}.RGB
   B=VectorParameter() {group=Mobile Grass, parameterName=Mobile Grass Variation 2, defaultValue=<rgba 0.5573 0.5332 0.5225 1>}.RGB
   Alpha=NamedRerouteUsage()
  )
  n3 = NamedRerouteUsage()
  n4 = Add(
   A=LinearInterpolate(
    A=VectorParameter() {group=Mobile Grass, parameterName=Mobile Grass Color 1, defaultValue=<rgba 0.035 0.1042 0.012 1>}.RGB
    B=VectorParameter() {group=Mobile Grass, parameterName=Mobile GRass Color 2, defaultValue=<rgba 0.0733 0.1354 0.0085 1>}.RGB
    Alpha=ComponentMask(None=@n1) {g=true}
   )
   B=Multiply(A=Add(A=@n2) {constB=-0.5}, B=ComponentMask(None=@n3) {r=true})
  )
  n5 = SetMaterialAttributes!(MaterialAttributes, BaseColor=Saturate(None=@n4)) {desc=Mobile Material, attributeSetTypes [#1]}
  n6 = NamedRerouteUsage()
  n7 = NamedRerouteUsage()
  n8 = NamedRerouteUsage()
  n9 = NamedRerouteUsage()
  n10 = NamedRerouteUsage()
  FunctionOutput(None=MaterialFunctionCall!(
    Material=MaterialFunctionCall!(GrassMaterial=@n5, Layer Mask, Frensel Color) {materialFunction=/Game/Athena/Helios/Environments/Landscape/Materials/MaterialFunctions_Utilities/MF_MobileGrass_Fresnel}.Grass Fresnel
    GrassEdgeMask
    TerrainMasks_QuadA=@n6
    TerrainMasks_QuadB=@n7
    TerrainMasks_QuadC=@n8
    TerrainMasks_QuadD=@n9
    GrassNoise
    VariationMasks Mobile=@n10
   ) {materialFunction=/Game/Hermes/Terrain/Landscape/Materials/Layers/BiomeColor/MF_Hermes_Grass_BiomeColor}.Result Mobile) {outputName=Grass Foliage Blend, sortPriority=2}
  n11 = SetMaterialAttributes!(
   MaterialAttributes
   BaseColor=NamedRerouteUsage()
   Specular=NamedRerouteUsage()
   Roughness=NamedRerouteUsage()
   Normal=NamedRerouteUsage()
   OpacityMask=NamedRerouteUsage()
  ) {
   desc=HighEnd Material
   attributeSetTypes [
    #1
    #2
    #3
    #4
    #5
   ]
  }
  n12 = MaterialFunctionCall(GrassMaterial=@n11) {materialFunction=&MF_MobileGrass_Fresnel}
  FunctionOutput(None=MaterialFunctionCall(
    Material=@n12.Grass Fresnel
    TerrainMasks_QuadA=@n6
    TerrainMasks_QuadB=@n7
    TerrainMasks_QuadC=@n8
    TerrainMasks_QuadD=@n9
    VariationMasks Mobile=@n10
   ) {materialFunction=&MF_Hermes_Grass_BiomeColor}.Result Mobile) {outputName=Static Meshes, sortPriority=1}
  FunctionOutput(None=MaterialFunctionCall(
    Material=QualitySwitch(Default=@n12.Grass Fresnel, Medium=@n11)
    TerrainMasks_QuadA=@n6
    TerrainMasks_QuadB=@n7
    TerrainMasks_QuadC=@n8
    TerrainMasks_QuadD=@n9
    VariationMasks Mobile=@n10
   ) {materialFunction=&MF_Hermes_Grass_BiomeColor}.Result Mobile)
  n13 = WorldPosition() {worldPositionShaderOffset=WPT_ExcludeAllShaderOffsets}
  n14 = TextureSample(UVs=Divide(
    A=TruncateLWC!(None=@n13.XY)
    B=ScalarParameter() {defaultValue=700, group=Grass, parameterName=LandscapeGrass_BaseScale, sortPriority=-1000}
   )) {samplerType=SAMPLERTYPE_Masks, texture=/Game/Environments/Helios/Terrain/TexturesMobile/T_Helios_Mobile_Grass}
  NamedRerouteDeclaration(None=@n14.A) {name=Height, nodeColor=<rgba 0.7344 0 1 1>}
  NamedRerouteDeclaration(None=LinearInterpolate(
    A=ScalarParameter() {defaultValue=0.7, group=Mobile Grass, parameterName=Mobile Grass Roughness Min}
    B=ScalarParameter() {defaultValue=0.3, group=Mobile Grass, parameterName=Mobile Grass Roughness Max}
    Alpha=NamedRerouteUsage()
   )) {name=GrassRoughness, nodeColor=<rgba 0.1484 0 1 1>}
  NamedRerouteDeclaration(None=FunctionInput(Preview=Constant()) {bUsePreviewValueAsDefault=true, inputName=TerrainMasks_QuadD, inputType=FunctionInput_Vector4, sortPriority=-100}) {name=TerrainMasks_QuadD, nodeColor=<rgba 0.5736 0.0077 1 1>}
  n15 = ScalarParameter() {defaultValue=0.06, group=Mobile Grass, parameterName=Mobile Grass Specular Min}
  n16 = ScalarParameter() {defaultValue=0.5, group=Mobile Grass, parameterName=Mobile Grass Specular Max}
  n17 = NamedRerouteUsage()
  NamedRerouteDeclaration(None=Saturate(None=LinearInterpolate(A=@n15, B=@n16, Alpha=@n17))) {name=GrassSpecular, nodeColor=<rgba 0.5 0 1 1>}
  n18 = Add(A=NamedRerouteUsage())
  NamedRerouteDeclaration(None=Saturate(None=Multiply(A=@n4, B=@n18))) {name=GrassBaseColor, nodeColor=<rgba 0.5 0 1 1>}
  NamedRerouteDeclaration(None=FunctionInput() {bUsePreviewValueAsDefault=true, inputName=VariationMasks Mobile, inputType=FunctionInput_Vector4, sortPriority=1}) {name=VariationMasks Mobile, nodeColor=<rgba 0.9219 1 0 1>}
  NamedRerouteDeclaration(None=@n14.RGB) {name=BladeNormals, nodeColor=<rgba 0.5 0 1 1>}
  NamedRerouteDeclaration(None=FunctionInput(Preview=Constant()) {bUsePreviewValueAsDefault=true, inputName=TerrainMasks_QuadA, inputType=FunctionInput_Vector4, sortPriority=-400}) {name=TerrainMasks_QuadA, nodeColor=<rgba 0.5736 0.0077 1 1>}
  NamedRerouteDeclaration(None=FunctionInput(Preview=Constant()) {bUsePreviewValueAsDefault=true, inputName=TerrainMasks_QuadB, inputType=FunctionInput_Vector4, sortPriority=-300}) {name=TerrainMasks_QuadB, nodeColor=<rgba 0.5736 0.0077 1 1>}
  NamedRerouteDeclaration(None=FunctionInput(Preview=Constant()) {bUsePreviewValueAsDefault=true, inputName=TerrainMasks_QuadC, inputType=FunctionInput_Vector4, sortPriority=-200}) {name=TerrainMasks_QuadC, nodeColor=<rgba 0.5736 0.0077 1 1>}
  n19 = NamedRerouteUsage()
  n20 = Saturate(None=Multiply(A=ComponentMask(None=@n19) {b=true}) {constB=4})
  NamedRerouteDeclaration(None=LinearInterpolate(A=@n14.A, B=@n14.A, Alpha=NamedRerouteDeclaration(None=@n20) {name=Contrasted Mask, nodeColor=<rgba 1 0 0.1406 1>})) {name=HeightMapFromTexture, nodeColor=<rgba 0.5 0 1 1>}
 ]
}
